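# Unlit translucent Niagara sprite material for the ground aura of the Energize Max LV3 emote effect.
# Panning radial UVs, warped by a distortion normal map, sample two blended noises that are sharpened and tinted by particle color for emissive.
# Opacity combines particle alpha with a distorted radial gradient, circular edge mask and hollowed center, softened by depth and view-angle fades.
# Parameters expose noise power, opacity, center hole shape and edge fade.
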Material /BRCosmetics/Effects/Fort_Effects/Emotes/Energize_Max/LV3/M_Energize_Max_Aura_LV3_ground
graph {
 properties {bUsedWithNiagaraSprites=true, blendMode=BLEND_Translucent, pixelDepthOffsetMode=PDOM_Legacy, shadingModel=MSM_Unlit, usedShadingModels=Unlit}
 nodes [
  n1 = ParticleColor()
  n2 = Constant2Vector() {g=1, r=2}
  n3 = MaterialFunctionCall() {materialFunction=&VectorToRadialValue}
  n4 = TextureCoordinate()
  n5 = TextureSample!(
   UVs=Panner(Coordinate=Multiply(A=@n2, B=@n3.Radial Coordinates)) {speedY=-0.5}
   Tex
   "DDX(UVs)"=DDX(Value=@n4)
   "DDY(UVs)"=DDY(Value=@n4)
   Apply View MipBias
  ) {mipValueMode=TMVM_Derivative, samplerType=SAMPLERTYPE_Normal, texture=&T_Weap_Distort_N_02}
  n6 = Multiply(A=ComponentMask(None=@n5.RGB) {g=true, r=true})
  n7 = Constant2Vector() {g=0.7, r=2}
  n8 = MaterialFunctionCall() {materialFunction=&VectorToRadialValue}
  n9 = Panner(Coordinate=Add(A=Multiply(A=@n6) {constB=0.3}, B=Multiply(A=@n7, B=@n8.Radial Coordinates))) {speedY=-1.5}
  n10 = TextureCoordinate()
  n11 = DDX(Value=@n10)
  n12 = DDY(Value=@n10)
  n13 = TextureSample(UVs=@n9, "DDX(UVs)"=@n11, "DDY(UVs)"=@n12) {mipValueMode=TMVM_Derivative, samplerType=SAMPLERTYPE_LinearColor, texture=&T_NoisePack_LinearColor}
  n14 = TextureSample(UVs=@n9, "DDX(UVs)"=@n11, "DDY(UVs)"=@n12) {
   mipValueMode=TMVM_Derivative
   samplerType=SAMPLERTYPE_LinearGrayscale
   texture=/Engine/Functions/Engine_MaterialFunctions02/ExampleContent/Textures/LowResBlurredNoise
  }
  n15 = Power(
   Base=LinearInterpolate(A=@n13.G, B=@n14.R, Alpha=Constant() {r=0.5}) {constAlpha=0.2}
   Exp=ScalarParameter() {defaultValue=3, parameterName=Noise power}
  ) {constExponent=3}
  MaterialFunctionCall(Emissive=Multiply(A=@n1.RGB, B=Multiply(A=@n15))) {materialFunction=&MF_FX_EmissiveAdjustments}
  n16 = Constant2Vector() {g=1, r=2}
  n17 = MaterialFunctionCall() {materialFunction=&VectorToRadialValue}
  n18 = TextureCoordinate()
  n19 = TextureSample(
   UVs=Panner(Coordinate=Multiply(A=@n16, B=@n17.Radial Coordinates)) {speedX=0.4, speedY=-2}
   "DDX(UVs)"=DDX(Value=@n18)
   "DDY(UVs)"=DDY(Value=@n18)
  ) {mipValueMode=TMVM_Derivative, samplerType=SAMPLERTYPE_Normal, texture=&T_Weap_Distort_N_02}
  n20 = Add(A=TextureCoordinate(), B=Multiply(A=@n19.R) {constB=0.7})
  n21 = Multiply(
   A=Multiply(A=MaterialFunctionCall(UVs=@n20) {materialFunction=&RadialGradientExponential}.RadialGradientExponential)
   B=ScalarParameter() {defaultValue=5, parameterName=Opacity}
  ) {constB=5}
  n22 = TextureCoordinate()
  n23 = Constant2Vector() {g=0.5, r=0.5}
  n24 = DepthFade(Opacity=Multiply(A=Saturate(None=@n21), B=SphereMask(A=@n22, B=@n23) {attenuationRadius=0.5, hardnessPercent=0})) {fadeDistanceDefault=30}
  n25 = OneMinus(None=SphereMask(A=TextureCoordinate(), B=Constant2Vector() {g=0.5, r=0.5}, Radius=ScalarParameter() {defaultValue=0.1, parameterName=Radius}) {attenuationRadius=0.1, hardnessPercent=0})
  n26 = ScalarParameter() {defaultValue=1, parameterName=Center mask power}
  Multiply(
   A=Multiply(A=Multiply(A=@n1.A, B=@n24), B=Power(Base=@n25, Exp=@n26) {constExponent=1})
   B=MaterialFunctionCall("FadeSensitivity (optional)"=ScalarParameter() {parameterName=Edge fade}) {materialFunction=&FadeBasedOnViewAngle}.Result
  )
 ]
}
